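# Material function that gives the Underground Rebel toon skin comic-book shading, toggled by a static switch.
# The directional light, read through a Custom node, drives biased shadow, specular and rim-light masks, varied by an SSC mask and Voronoi texture.
# Hatching comes from a camera-facing virtual plane or Voronoi-distorted UVs, while a distance-field input draws depth-adjusted lines.
# It blacks out base color, feeds emissive and applies LUT and time-of-day color adjustments.
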MaterialFunction /Game/Characters/Player/Male/Medium/Bodies/M_MED_UndergroundRebel_Fashion/Skins/Toon/Materials/MF_UndergroundRebel_ComicShading
graph {nodes [
  n1 = Rotator(Coordinate=Add(A=ScreenPosition().ViewportUV) {constB=-0.5}, Time=Constant() {r=0.125}) {centerX=0, centerY=0, speed=-6.283}
  n2 = Frac(None=Multiply(A=ComponentMask(None=@n1) {r=true}) {constB=64})
  Abs(None=Add(A=Multiply(A=@n2) {constB=2}) {constB=-1}) {desc=HatchingDF}
  FunctionOutput(None=StaticSwitchParameter(True=NamedRerouteUsage(), False=NamedRerouteUsage()) {group=01_UndergroundRebel_ComicShading, parameterName=UseUndergroundRebel_ComicShading, sortPriority=0})
  n3 = NamedRerouteUsage()
  n4 = ScalarParameter() {defaultValue=2, group=01_UndergroundRebel_ComicShading, parameterName=ShadedBaseColor_Brightness, sliderMax=2}
  LinearInterpolate(A=@n3, B=Saturate(None=Multiply(A=@n3, B=@n4)), Alpha=NamedRerouteUsage())
  n5 = NamedRerouteUsage()
  n6 = Power(
   Base=NamedRerouteUsage()
   Exp=ScalarParameter() {defaultValue=6, group=01_UndergroundRebel_ComicShading, parameterName=BackLit_Exponent, sliderMax=8}
  ) {constExponent=4}
  n7 = NamedRerouteDeclaration(None=Saturate(None=SmoothStep(Value=@n6) {constMax=0, constMin=0.02})) {name=RimLight_Mask, nodeColor=<rgba 0 1 0.5078 1>}
  n8 = DotProduct(
   A=MaterialFunctionCall() {materialFunction=&CameraDirectionVector}.Result
   B=NamedRerouteUsage()
  )
  n9 = ScalarParameter() {defaultValue=0.25, group=01_UndergroundRebel_ComicShading, parameterName=BackLit_Bias, sliderMax=1}
  n10 = NamedRerouteDeclaration(None=Clamp(
    None=Add(A=@n8, B=@n9)
    Max=ScalarParameter() {defaultValue=0.75, group=01_UndergroundRebel_ComicShading, parameterName=BackLit_MaxBias, sliderMax=1}
   )) {name=BackLit_Curve, nodeColor=<rgba 0.9507 0.45 1 1>}
  LinearInterpolate(A=@n5, B=Add(A=@n5, B=@n7), Alpha=@n10)
  n11 = StaticSwitchParameter(True=NamedRerouteUsage(), False=NamedRerouteUsage()) {parameterName=UseScreenSpaceHatching}
  n12 = Constant() {r=0.5}
  n13 = Multiply(A=Saturate(None=ScalarParameter() {defaultValue=0.5, group=01_UndergroundRebel_ComicShading, parameterName=Hatching_TransitionAmount, sliderMax=1})) {constB=0.5}
  n14 = Multiply(A=Subtract(
    A=NamedRerouteUsage()
    B=ScalarParameter() {defaultValue=0.6, group=01_UndergroundRebel_ComicShading, parameterName=Hatching_Bias, sliderMax=1, sliderMin=-1}
   )) {constB=0.5}
  n15 = SmoothStep(
   Min=Subtract(A=@n12, B=@n13) {constA=0.5}
   Max=Add(A=@n12, B=@n13) {constA=0.5, constB=-0.5}
   Value=MaterialFunctionCall(Base=NamedRerouteUsage(), Blend=Add(A=@n14) {constB=0.5}) {materialFunction=&Blend_Overlay}.Result
  ) {constMax=0.01}
  Multiply(A=Saturate(None=Add(A=@n11, B=@n15)))
  n16 = VectorParameter() {group=01_UndergroundRebel_DistanceFieldLines, parameterName=ShadowColor, sortPriority=3, defaultValue=<rgba 0 0 0 1>}
  n17 = NamedRerouteUsage()
  n18 = LinearInterpolate(
   A=Multiply(
    A=ScalarParameter() {defaultValue=0.25, group=01_UndergroundRebel_ComicShading, parameterName=ShadedBaseColor_Intensity, sliderMax=1}
    B=@n17
   )
   B=@n17
   Alpha=NamedRerouteUsage()
  )
  n19 = NamedRerouteUsage()
  n20 = NamedRerouteUsage()
  n21 = NamedRerouteUsage()
  NamedRerouteDeclaration(None=LinearInterpolate(A=@n16.RGB, B=Add(A=@n18, B=@n19), Alpha=Multiply(A=@n20, B=@n21))) {name=BaseColor_Raw, nodeColor=<rgba 1 0.4506 0.2387 1>}
  n22 = MaterialFunctionCall(Float2=Add(A=ScreenPosition().ViewportUV) {constB=-0.5}) {materialFunction=&BreakOutFloat2Components}
  n23 = MaterialFunctionCall(Float2=ViewSize()) {materialFunction=&BreakOutFloat2Components}
  n24 = Divide(A=@n23.R, B=@n23.R)
  NamedRerouteDeclaration(None=AppendVector(A=Multiply(A=@n22.R, B=@n24), B=Multiply(A=@n22.G) {constB=-1})) {name=ScreenUVs, nodeColor=<rgba 0.1484 0 1 1>}
  n25 = DotProduct(
   A=Rotator(Coordinate=NamedRerouteUsage(), Time=ScalarParameter() {group=01_UndergroundRebel_ComicShading, parameterName=Lines_Rotation, sliderMax=1}) {centerX=0, centerY=0, speed=-6.283}
   B=NamedRerouteUsage()
  )
  n26 = ScalarParameter() {defaultValue=96, group=01_UndergroundRebel_ComicShading, parameterName=Lines_Amount}
  n27 = Multiply(A=Frac(None=Multiply(A=@n25, B=@n26) {constB=32})) {constB=2}
  NamedRerouteDeclaration(None=Abs(None=Add(A=@n27) {constB=-1}) {desc=HatchingDF}) {name=VirtualPlaneHatching, nodeColor=<rgba 0.9922 1 0 1>}
  n28 = Multiply(
   A=NamedRerouteUsage()
   B=ScalarParameter() {defaultValue=0.5, group=01_UndergroundRebel_ComicShading, parameterName=LineDistortion_RandScale, sliderMax=1}
  )
  n29 = NamedRerouteUsage()
  n30 = ScalarParameter() {defaultValue=0.5, group=01_UndergroundRebel_ComicShading, parameterName=LineDistortion_RandRot, sliderMax=0.25}
  n31 = Rotator(
   Coordinate=Multiply(A=TextureCoordinate(), B=Add(A=@n28))
   Time=Add(A=ScalarParameter() {group=01_UndergroundRebel_ComicShading, parameterName=Lines_RotationOffset, sliderMax=1}, B=Multiply(A=@n29, B=@n30))
  ) {centerX=0, centerY=0, speed=-6.283}
  n32 = Frac(None=Multiply(
    A=ComponentMask(None=@n31) {r=true}
    B=ScalarParameter() {defaultValue=180, group=01_UndergroundRebel_ComicShading, parameterName=UV_Lines_Amount}
   ) {constB=32})
  n33 = Abs(None=Add(A=Multiply(A=@n32) {constB=2}) {constB=-1}) {desc=HatchingDF}
  n34 = NamedRerouteUsage()
  NamedRerouteDeclaration(None=Saturate(None=LinearInterpolate(B=@n33, Alpha=@n34) {constA=1})) {name=UVHatching, nodeColor=<rgba 0.9922 1 0 1>}
  n35 = TransformPosition(None=WorldPosition().XYZ) {transformSourceType=TRANSFORMPOSSOURCE_World, transformType=TRANSFORMPOSSOURCE_Camera}
  n36 = VertexNormalWS()
  n37 = MaterialFunctionCall(XY=CrossProduct(A=Normalize(VectorInput=@n35), B=Transform(None=@n36) {transformSourceType=TRANSFORMSOURCE_World, transformType=TRANSFORM_Camera})) {materialFunction=&Swizzle}
  NamedRerouteDeclaration(None=Add(A=Multiply(A=@n37.YX) {constB=0.5}) {constB=0.5}) {name=MatCapUVs, nodeColor=<rgba 1 0.867 0 1>}
  n38 = MaterialFunctionCall(Float2=TextureCoordinate()) {materialFunction=&BreakOutFloat2Components}
  n39 = Multiply(
   A=@n38.G
   B=Divide(B=ScalarParameter() {defaultValue=1, group=01_UndergroundRebel_ComicShading, parameterName=ShadowVariation_Ratio, sliderMax=2})
  )
  n40 = TextureSample(UVs=Multiply(
    A=MaterialFunctionCall(X=@n38.R, Y=@n39) {materialFunction=&MakeFloat2}.Result
    B=ScalarParameter() {defaultValue=4, group=01_UndergroundRebel_ComicShading, parameterName=ShadowVariation_Tiling, sliderMax=16}
   )) {samplerType=SAMPLERTYPE_LinearColor, texture=/Game/Global/Materials/Character_Parent_2023/Anime_Parent_2023/Textures/T_Hatching_Voronoi}
  NamedRerouteDeclaration(None=@n40.G) {name=VoronoiID, nodeColor=<rgba 1 0.2603 0.2703 1>}
  n41 = NamedRerouteUsage()
  n42 = DotProduct(A=ReflectionVectorWS(), B=NamedRerouteUsage()) {desc=Phong}
  n43 = Multiply(
   A=SphereMask(
    A=Saturate(None=@n42)
    B=Constant() {r=1}
    Radius=ScalarParameter() {defaultValue=0.1, group=01_UndergroundRebel_ComicShading, parameterName=Specular_Size, sliderMax=1}
    Hardness=ScalarParameter() {defaultValue=0.8, group=01_UndergroundRebel_ComicShading, parameterName=Specular_Hardness, sliderMax=1}
   ) {hardnessPercent=78}
   B=ScalarParameter() {defaultValue=0.5, group=01_UndergroundRebel_ComicShading, parameterName=Specular_Intensity, sliderMax=1}
  )
  NamedRerouteDeclaration(None=Saturate(None=Multiply(A=@n41, B=@n43))) {name=Specular_Curve, nodeColor=<rgba 1 0.4453 0 1>}
  n44 = TextureSample() {
   samplerType=SAMPLERTYPE_Masks
   texture=/Game/Characters/Player/Male/Medium/Bodies/M_MED_UndergroundRebel_Fashion/Skins/Toon/Textures/T_UndergroundRebel_FashionToon_Body_SSC
  }
  n45 = MaterialFunctionCall(Float3=NamedRerouteDeclaration(None=FunctionInput(Preview=@n44.RGB) {inputName=SSC, sortPriority=2}) {name=SSC, nodeColor=<rgba 0 0.5 1 1>}) {materialFunction=&BreakOutFloat3Components}
  NamedRerouteDeclaration(None=@n45.B) {name=Sheen, nodeColor=<rgba 0 0.5 1 1>}
  n46 = NamedRerouteUsage()
  n47 = ScalarParameter() {defaultValue=24, group=01_UndergroundRebel_ComicShading, parameterName=LineDistortion_Tiling, sliderMax=128}
  n48 = VectorNoise(World Position=AppendVector(A=Multiply(A=@n46, B=@n47), B=Constant())) {noiseFunction=VNF_VoronoiALU}
  n49 = ComponentMask(None=@n48) {b=true, g=true, r=true}
  NamedRerouteDeclaration(None=Add(A=VectorNoise(World Position=@n49)) {constB=-0.5}) {name=Voronoi, nodeColor=<rgba 1 0.2257 0.2955 1>}
  NamedRerouteDeclaration(None=FunctionInput() {inputName=PreSkinnnedLocalPosition, sortPriority=1}) {name=PreSkinnedLocalPosition, nodeColor=<rgba 0 1 0.7422 1>}
  n50 = ComponentMask(None=@n48) {a=true}
  NamedRerouteDeclaration(None=Saturate(None=InverseLinearInterpolate(Value=@n50) {constA=0.5})) {name=VoronoiFadeing, nodeColor=<rgba 1 0.2603 0.2703 1>}
  n51 = Add(A=NamedRerouteUsage(), B=CameraVectorWS())
  NamedRerouteDeclaration(None=DotProduct(A=Normalize(VectorInput=@n51), B=PixelNormalWS())) {name=NdotH, nodeColor=<rgba 0.7344 1 0 1>}
  n52 = Constant() {r=0.5}
  n53 = Multiply(A=Saturate(None=ScalarParameter() {defaultValue=0.5, group=01_UndergroundRebel_ComicShading, parameterName=ShadedBaseColor_TransitionAmount, sliderMax=1})) {constB=0.5}
  n54 = NamedRerouteUsage()
  n55 = ScalarParameter() {group=01_UndergroundRebel_ComicShading, parameterName=ShadedBaseColor_Bias, sliderMax=1, sliderMin=1}
  n56 = Add(A=Multiply(A=Subtract(A=@n54, B=@n55)) {constB=0.5}) {constB=0.5}
  n57 = NamedRerouteUsage()
  NamedRerouteDeclaration(None=SmoothStep(Min=Subtract(A=@n52, B=@n53) {constA=0.5}, Max=Add(A=@n52, B=@n53) {constA=0.5, constB=-0.5}, Value=LinearInterpolate(A=@n56, Alpha=@n57)) {constMax=0.01}) {name=ShadedBaseColor_Curve, nodeColor=<rgba 0 0.9439 1 1>}
  n58 = TextureSample() {
   samplerType=SAMPLERTYPE_Alpha
   texture=/Game/Characters/Player/Male/Medium/Bodies/M_MED_UndergroundRebel_Fashion/Skins/Toon/Textures/T_UndergroundRebel_FashionToon_Body_DF
  }
  NamedRerouteDeclaration(None=FunctionInput(Preview=@n58.R) {inputName=DF, inputType=FunctionInput_Scalar, sortPriority=3}) {name=DF, nodeColor=<rgba 0 1 0.6166 1>}
  n59 = ScalarParameter() {defaultValue=0.4, group=01_UndergroundRebel_DistanceFieldLines, parameterName=DF_LineThickness_Near, sliderMax=1, sortPriority=5}
  n60 = ScalarParameter() {defaultValue=0.6, group=01_UndergroundRebel_DistanceFieldLines, parameterName=DF_LineThickness_Far, sliderMax=1, sortPriority=6}
  n61 = Divide(
   A=Subtract(
    A=PixelDepth().R
    B=ScalarParameter() {defaultValue=50, group=01_UndergroundRebel_DistanceFieldLines, parameterName=DF_FadeStart, sliderMax=200}
   )
   B=ScalarParameter() {defaultValue=500, group=01_UndergroundRebel_DistanceFieldLines, parameterName=DF_FadeDistance, sliderMax=1000}
  )
  n62 = ScalarParameter() {defaultValue=0.9, group=01_UndergroundRebel_DistanceFieldLines, parameterName=DF_LineHardness, sliderMax=1, sortPriority=7}
  n63 = SphereMask(A=NamedRerouteUsage(), B=Constant() {r=1}, Radius=LinearInterpolate(A=@n59, B=@n60, Alpha=Saturate(None=@n61)), Hardness=@n62) {attenuationRadius=0.1}
  NamedRerouteDeclaration(None=OneMinus(None=Saturate(None=@n63))) {name=DF_Mask, nodeColor=<rgba 1 0.6562 0 1>}
  n64 = Transform(None=NamedRerouteDeclaration(None=Custom() {code="return View.DirectionalLightDirection;", description=DirectionalLightDirection, inputs []}) {name=LightVector, nodeColor=<rgba 0.4766 1 0 1>}) {transformSourceType=TRANSFORMSOURCE_World, transformType=TRANSFORM_View}
  NamedRerouteDeclaration(None=Normalize(VectorInput=ComponentMask(None=@n64) {g=true, r=true})) {name=LightVector_ScreenSpace, nodeColor=<rgba 1 0 0.9844 1>}
  n65 = MaterialFunctionCall!(In=NamedRerouteUsage()) {materialFunction=/Game/Packages/Fortress_SharedMaterials/Anime_Master_Material/MF_TODCustomColors}
  NamedRerouteDeclaration(None=MaterialFunctionCall!(Color=MaterialFunctionCall!(In=@n65.Result, LUT, LUT Multiplier) {materialFunction=/Game/Packages/Fortress_SharedMaterials/Anime_Master_Material/MF_ApplyLUT}.Result) {materialFunction=/Game/Packages/Fortress_SharedMaterials/Anime_Master_Material/MF_TODColorAdjustment}.Result) {name=BaseColor_Out, nodeColor=<rgba 1 0.9844 0 1>}
  NamedRerouteDeclaration(None=DotProduct(A=CameraVectorWS(), B=PixelNormalWS())) {name=NdotV, nodeColor=<rgba 0.7344 1 0 1>}
  n66 = Constant()
  n67 = NamedRerouteUsage()
  n68 = NamedRerouteUsage()
  n69 = SetMaterialAttributes!(
   MaterialAttributes=NamedRerouteUsage()
   BaseColor=@n66
   Specular=@n66
   Metallic=NamedRerouteUsage()
   Roughness=NamedRerouteUsage()
   EmissiveColor=StaticSwitchParameter(True=NamedRerouteUsage(), False=Add(A=@n67, B=@n68)) {group=01_UndergroundRebel_ComicShading, parameterName=DebugHatching, sortPriority=99}
   ShadingModel=ShadingModel() {shadingModel=MSM_Cloth}
  ) {attributeSetTypes [
    #1
    #2
    #3
    #4
    #5
    #6
   ]}
  NamedRerouteDeclaration(None=@n69) {name=MA_Out, nodeColor=<rgba 0 0.4375 1 1>}
  n70 = Constant() {r=0.5}
  n71 = Multiply(A=Saturate(None=ScalarParameter() {defaultValue=0.5, group=01_UndergroundRebel_ComicShading, parameterName=Shadow_TransitionAmount, sliderMax=1})) {constB=0.5}
  n72 = Multiply(A=Subtract(A=NamedRerouteUsage(), B=ScalarParameter() {group=01_UndergroundRebel_ComicShading, parameterName=Shadow_Bias, sliderMax=1, sliderMin=-1})) {constB=0.5}
  n73 = NamedRerouteUsage()
  n74 = ScalarParameter() {defaultValue=0.2, group=01_UndergroundRebel_ComicShading, parameterName=ShadowVariation_Intensity, sliderMax=1}
  n75 = Add(A=Multiply(A=Add(A=@n72) {constB=0.5}, B=Constant() {r=2}), B=Subtract(A=Multiply(A=@n73, B=@n74) {constB=0.2}, B=Constant() {r=1}))
  n76 = SmoothStep(
   Min=Subtract(A=@n70, B=@n71) {constA=0.5}
   Max=Add(A=@n70, B=@n71) {constA=0.5, constB=-0.5}
   Value=MaterialFunctionCall(Base=NamedRerouteUsage(), Blend=Saturate(None=@n75) {desc=height blend with 1 heighmap}) {materialFunction=&Blend_Overlay}.Result
  ) {constMax=0.01}
  n77 = NamedRerouteUsage()
  NamedRerouteDeclaration(None=Saturate(None=LinearInterpolate(A=@n76, Alpha=@n77))) {name=ShadingMask, nodeColor=<rgba 1 0.6562 0 1>}
  NamedRerouteDeclaration(None=Multiply(A=@n7, B=@n10)) {name=BackLitShading_Curve, nodeColor=<rgba 0.1307 1 0.334 1>}
  n78 = SetMaterialAttributes!(MaterialAttributes, BaseColor=Constant() {r=0.18}) {attributeSetTypes [#1]}
  n79 = GetMaterialAttributes(None=NamedRerouteDeclaration(None=FunctionInput(Preview=@n78) {inputName=MA, inputType=FunctionInput_MaterialAttributes}) {name=MA_In, nodeColor=<rgba 0 0.6484 1 1>}) {attributeGetTypes [#1, #5]}
  NamedRerouteDeclaration(None=@n79.EmissiveColor) {name=Emissive_In, nodeColor=<rgba 1 0 0.6797 1>}
  NamedRerouteDeclaration(None=SmoothStep(Value=@n40.R) {constMax=0.8, constMin=1}) {name=VoronoiIEdgeDistance, nodeColor=<rgba 1 0.2603 0.2703 1>}
  NamedRerouteDeclaration(None=@n40.B) {name=VoronoiICenterDistance, nodeColor=<rgba 1 0.2603 0.2703 1>}
  n80 = Constant3Vector() {constant=<rgba 1 1 0 0>}
  n81 = MaterialFunctionCall() {materialFunction=&CameraDirectionVector}
  n82 = Multiply(B=Normalize(VectorInput=Multiply(A=@n80, B=@n81.Result))) {constA=0.25}
  n83 = ScalarParameter() {defaultValue=10, group=01_UndergroundRebel_ComicShading, parameterName=ScreenPostionDepthBias, sliderMax=100}
  n84 = VectorParameter() {group=01_UndergroundRebel_ComicShading, parameterName=VirtualPlanePivotOffset, defaultValue=<rgba 0 0 90 1>}
  n85 = Add(A=ActorPositionWS(), B=@n84.RGB)
  n86 = Constant() {r=45}
  n87 = Constant() {r=90}
  n88 = MaterialFunctionCall() {materialFunction=/Engine/Functions/Engine_MaterialFunctions02/Utility/FOV}
  n89 = MaterialFunctionCall(
   Plane Normal Axis=Normalize(VectorInput=Add(A=@n82, B=@n81.Result))
   Plane Center=Add(
    A=Multiply(A=@n81.Result, B=@n83) {constB=50}
    B=NamedRerouteDeclaration(None=@n85) {name=AdjustedActorPosition, nodeColor=<rgba 0.8281 0 1 1>}
   )
   UV Scale=LinearInterpolate(A=Constant() {r=64}, B=Constant() {r=128}, Alpha=InverseLinearInterpolate(A=@n86, B=@n87, Value=@n88.FOV angle in degrees))
  ) {materialFunction=&VirtualPlaneCoordinates}
  n90 = MaterialFunctionCall(Float2=Add(A=NamedRerouteDeclaration(None=@n89.UVs) {name=ObjectAlighedScreenPosition, nodeColor=<rgba 1 0.867 0 1>}) {constB=-0.5}) {materialFunction=&BreakOutFloat2Components}
  n91 = MaterialFunctionCall(Float2=ViewSize()) {materialFunction=&BreakOutFloat2Components}
  n92 = Divide(A=@n91.R, B=@n91.R)
  NamedRerouteDeclaration(None=AppendVector(A=Multiply(A=@n90.R, B=@n92), B=Multiply(A=@n90.G) {constB=-1})) {name=ObjectAlighedScreenUVs, nodeColor=<rgba 1 0.867 0 1>}
  NamedRerouteDeclaration(None=DotProduct(A=NamedRerouteUsage(), B=PixelNormalWS())) {name=NdotL, nodeColor=<rgba 0.7344 1 0 1>}
  n93 = Saturate(None=ScalarParameter() {defaultValue=1, group=01_UndergroundRebel_ComicShading, parameterName=Threshold_Intensity, sliderMax=1})
  n94 = Power(Base=@n45.R, Exp=ScalarParameter() {defaultValue=1, group=01_UndergroundRebel_ComicShading, parameterName=Threshold_Exponent, sliderMax=8}) {constExponent=1}
  NamedRerouteDeclaration(None=LinearInterpolate(A=OneMinus(None=@n93), Alpha=Saturate(None=@n94))) {name=Threshold, nodeColor=<rgba 0 0.5 1 1>}
  n95 = ActorPositionWS()
  n96 = CameraPositionWS()
  n97 = Divide(A=Subtract(A=PixelDepth().R, B=Distance(A=@n95, B=@n96))) {constB=1024}
  NamedRerouteDeclaration(None=Saturate(None=Abs(None=@n97))) {name=PixelDepth_ObjectAlighnt, nodeColor=<rgba 0.875 1 0 1>}
  NamedRerouteDeclaration(None=@n79.BaseColor) {name=BaseColor_In, nodeColor=<rgba 1 0 0.6797 1>}
  NamedRerouteUsage()
  NamedRerouteUsage()
  NamedRerouteUsage()
  NamedRerouteUsage()
  NamedRerouteUsage()
  OneMinus(None=@n34)
  SmoothStep() {constMax=0.7, constMin=0.9}
  n98 = ScalarParameter() {defaultValue=0.05, group=01_UndergroundRebel_ComicShading, parameterName=Hatching_Thickness, sliderMax=0.5}
  SmoothStep(Min=Subtract(B=@n98) {constA=0.5, constB=0.5}, Max=Add(A=@n98) {constB=0.5}) {constMax=0.5, constMin=0.4}
  n99 = CurveAtlasRowParameter() {
   atlas=/Game/Characters/Shared/CurveAtlas/AthenaCharacterCurveAtlas
   curve=/Game/Characters/Shared/CurveAtlas/HB_TwoToneSoftShortShadingCurve
   defaultValue=29
   group=01_UndergroundRebel_ComicShading
   parameterName=ShadingCurve
   sortPriority=5
  }
  StaticSwitchParameter(True=ChannelMaskParameter(None=@n99) {group=01_UndergroundRebel_ComicShading, parameterName=ShadingCurveChannel, sortPriority=10}.R) {group=01_UndergroundRebel_ComicShading, parameterName=UseShadingCurveAtlas, sortPriority=1}
 ]}
