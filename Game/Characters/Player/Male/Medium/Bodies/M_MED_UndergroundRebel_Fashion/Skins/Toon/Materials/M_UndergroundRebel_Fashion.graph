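# Character body material for the Underground Rebel Fashion toon skin, with cloth, skeletal, morph and mesh-deformer usage and a per-expression shading model.
# The shared 2023 character PreFX function feeds a skin-specific comic shading function, which also takes SSC mask and DF texture parameters.
# The comic-shaded attributes then pass through the shared PostFX function.
# A composite subgraph supplies pre-skinned position, distance blend and Fresnel inputs.
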Material /Game/Characters/Player/Male/Medium/Bodies/M_MED_UndergroundRebel_Fashion/Skins/Toon/Materials/M_UndergroundRebel_Fashion
graph {
 properties {
  bAutomaticallySetUsageInEditor=false
  bHasPixelAnimation=true
  bUseMaterialAttributes=true
  bUsedWithClothing=true
  bUsedWithMeshDeformer=true
  bUsedWithMorphTargets=true
  bUsedWithSkeletalMesh=true
  pixelDepthOffsetMode=PDOM_Legacy
  shadingModel=MSM_FromMaterialExpression
  usedShadingModels=Default Lit | Subsurface | Clear Coat | Subsurface Profile | Cloth
 }
 nodes [
  Composite() {subgraphName=Construct Base Inputs}
  n1 = MaterialFunctionCall() {materialFunction=&MF_CharacterParent_2023_PreFX}
  n2 = TextureSampleParameter2D() {
   group=00_Texture Inputs
   parameterName=SSC
   samplerType=SAMPLERTYPE_Masks
   sortPriority=1
   texture=/Game/Global/Textures/Default/Blanks/T_MidgreyMask
  }
  n3 = TextureSampleParameter2D() {
   group=00_Texture Inputs
   parameterName=DF
   samplerType=SAMPLERTYPE_Alpha
   sortPriority=1
   texture=/Game/Global/Textures/Default/Blanks/T_BlackAlpha
  }
  n4 = MaterialFunctionCall!(MA=@n1.Result, PreSkinnnedLocalPosition=@n1.Result, SSC=@n2.RGB, DF=@n3.R) {materialFunction=&/MF_UndergroundRebel_ComicShading}
  MaterialFunctionCall(Material=@n4.Result, M=@n1.M, Pre Skinned Local Position=@n1.M, Distance Blend=@n1.M, Fresnel=@n1.M) {materialFunction=&MF_CharacterParent_2023_PostFX}
  PinBase()
  n5 = Constant()
  n6 = PreSkinnedPosition()
  n7 = ShaderStageSwitch(PixelShader=VertexInterpolator(VS=@n6).PS, VertexShader=@n6)
  n8 = MaterialFunctionCall!(Material) {materialFunction=/Game/Packages/Fortress_SharedMaterials/Base_Material_Functions/CharacterBrightnessandScale_Balance}
  n9 = Fresnel() {baseReflectFraction=0, exponent=1}
  PinBase!(M=@n5, Pre Skinned Local Position=@n7, Distance Blend=@n8.Distance Blend, Fresnel=@n9) {reroutePins [
    M {expression=&self:MaterialExpressionReroute_4}
    Pre Skinned Local Position {expression=&self:MaterialExpressionReroute_5}
    Distance Blend {expression=&self:MaterialExpressionReroute_6}
    Fresnel {expression=&self:MaterialExpressionReroute_7}
   ]}
  MaterialExpressionReroute_4 = Reroute(None=@n5)
  MaterialExpressionReroute_5 = Reroute(None=@n7)
  MaterialExpressionReroute_6 = Reroute(None=@n8.Distance Blend)
  MaterialExpressionReroute_7 = Reroute(None=@n9)
 ]
}
